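# Animation Blueprint for the guided missile rocket mesh, driving the RC rocket skeleton from the rocket launcher weapons.
# The AnimGraph plays the Launcher_RC_IdleLoaded_CMM_W sequence as the base pose, passed through DefaultSlot so montages can override it before reaching the output pose.
# The Event Graph holds only an unconnected update event, and the blueprint has no variables or state logic.
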AnimBlueprint /Game/Animation/Game/MainPlayer/Combat/Guns/GuidedMissile/RC_Rocket_BP
target_skeleton=/Game/Weapons/FORT_RocketLaunchers/Mesh/RC_Rocket_Export_Skeleton
graphs [
 AnimGraph {nodes [Output Pose(Result=Slot 'DefaultSlot'(Source=Sequence Player 'Launcher_RC_IdleLoaded_CMM_W'() {sequence=&/Launcher_RC_IdleLoaded_CMM_W})) {layerGroup=None}]}
 EventGraph {nodes [Event BlueprintUpdateAnimation()]}
]
variables []
